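# Niagara idle effect for the CopperToll backpack cosmetic, adding green lantern and middle glows.
# Burst-spawned, persistent particles follow sampled locations on User.SkeletalMesh with local offsets.
# Lantern_Light_Glow pairs a sprite and light, with matching curves pulsing sprite alpha and light color over a looping lifetime.
# Middle_Glow spreads two sprites by exec index, while Middle_Light adds a constant light; Middle_Glow's alpha curve is disabled.
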NiagaraSystem /BRCosmetics/Accessories/FORT_Backpacks/Backpack_M_MED_CopperToll/FX/NS_Backpack_CopperToll_Idle
dependencies {
 usedDataInterfaces [&NiagaraDataInterfaceCurve]
 usedDynamicInputs [&RandomRangeFloat]
 usedModules [&SystemState]
}
emitters {
 Lantern_Light_Glow {
  emitterUpdateScript [
   &EmitterState()
   &SpawnBurst_Instantaneous(Use Spawn Probability=false, UseLoopCountLimit=true)
  ]
  particleSpawnScript [
   &InitializeParticle(
    Color=<rgba 0.0252 0.4125 0.2016 0.6>
    Lifetime=5
    Sprite Size Mode=::NewEnumerator3
    Uniform Sprite Size=14
    UsePositionOffset=false
    Write Material Random=true
   )
   &FN_Light_v2(Light Exponent=3, Light Radius=25, LightEnabled=true)
   &DynamicMaterialParameters(
    Index 0 Param 1=2
    Index 0 Param 2=0.25
    Index 0 Param 4=100
    Module.Index0Param0Write=true
    Module.Index0Param1Write=true
    Module.Index0Param2Write=true
    Module.Index0Param3Write=true
   )
   &CameraOffset(Camera Offset Amount=10)
  ]
  particleUpdateScript [
   &ParticleState(
    Kill Particles When Lifetime Has Elapsed=false
    Let Infinitely Lived Particles Die When Emitter Deactivates=true
    Loop Particles Lifetime=true
   )
   &SkeletalMeshLocation(
    Bone / Socket Index=1
    Bone Sampling Mode=::NewEnumerator4
    MaskSpawnGroup=false
    Sampled Orientation Coordinate Space=::World
    Sampled Position Coordinate Space=::World
    Sampled Position Offset=<3.25 0 0>
    Sampled Velocity Coordinate Space:ENiagaraCoordinateSpace=::World
    Skeletal Mesh=$User.SkeletalMesh
   )
   &ScaleColor(
    Scale Alpha=&FloatFromCurve(FloatCurve=DataInterface() {propertyValues {Curve {keys [
         <at 0.1 1>
         {arriveTangent=-0.7039, interpMode=RCIM_Cubic, leaveTangent=-0.7039, time=0.3472, value=1}
         <at 0.5972 0.65 RCIM_Cubic>
         <at 0.8472 1>
         <at 1.0972 0.65 RCIM_Cubic>
         <at 1.3472 1 RCIM_Cubic>
         <at 1.5972 0.65>
         <at 1.8472 1 RCIM_Cubic>
        ]}}})
    ScaleA=true
    ScaleRGB=true
   )
   /SetVariables(Particles.Light Color:LinearColor=&MultiplyLinearColor(
     Linear Color=<rgba 0.0252 0.4125 0.2016 1>
     Scale Factor=&MakeVector4_FromFloat(Float=&FloatFromCurve(
       FloatCurve=DataInterface() {propertyValues {Curve {keys [
           <at 0.1 1>
           {arriveTangent=-0.7039, interpMode=RCIM_Cubic, leaveTangent=-0.7039, time=0.3472, value=1}
           <at 0.5972 0.65 RCIM_Cubic>
           <at 0.8472 1>
           <at 1.0972 0.65 RCIM_Cubic>
           <at 1.3472 1 RCIM_Cubic>
           <at 1.5972 0.65>
           <at 1.8472 1 RCIM_Cubic>
          ]}}}
       Scale Curve=10
      ))
    )) {isSetParameters=true, module=SetVariables_#1}
  ]
  rendererClasses [&NiagaraLightRendererProperties, &NiagaraSpriteRendererProperties]
  renderers [
   {rendererClass=&NiagaraLightRendererProperties, rendererIndex=0}
   {rendererClass=&NiagaraSpriteRendererProperties, rendererIndex=1}
  ]
 }
 Middle_Glow {
  emitterUpdateScript [
   &EmitterState()
   &SpawnBurst_Instantaneous(Spawn Count=2, Use Spawn Probability=false, UseLoopCountLimit=true)
  ]
  particleSpawnScript [
   &InitializeParticle(
    Color=<rgba 0.0299 0.4896 0.2392 1>
    Sprite Size Mode=::NewEnumerator3
    Uniform Sprite Size=24
    UsePositionOffset=false
    Write Material Random=true
   )
   &DynamicMaterialParameters(
    Index 0 Param 1=3.5
    Index 0 Param 2=0.5
    Index 0 Param 4=100
    Module.Index0Param0Write=true
    Module.Index0Param1Write=true
    Module.Index0Param2Write=true
    Module.Index0Param3Write=true
   )
   &CameraOffset(Camera Offset Amount=8)
  ]
  particleUpdateScript [
   &ParticleState(Kill Particles When Lifetime Has Elapsed=false, Let Infinitely Lived Particles Die When Emitter Deactivates=true)
   &SkeletalMeshLocation(
    Bone Sampling Mode=::NewEnumerator4
    MaskSpawnGroup=false
    Sampled Orientation Coordinate Space=::World
    Sampled Position Coordinate Space=::World
    Sampled Position Offset=&Lerp_Vector(A=<-4 1 -1.5>, B=<4 1 -1.5>, Alpha=&ReturnNormalizedExecIndex())
    Sampled Velocity Coordinate Space:ENiagaraCoordinateSpace=::World
    Skeletal Mesh=$User.SkeletalMesh
   )
   &ScaleColor(
    Scale Alpha=&FloatFromCurve(
     FloatCurve=DataInterface() {propertyValues {Curve {keys [
         <at 0.1 1>
         {arriveTangent=-0.7039, interpMode=RCIM_Cubic, leaveTangent=-0.7039, time=0.3472, value=1}
         <at 0.5972 0.65 RCIM_Cubic>
         <at 0.8472 1>
         <at 1.0972 0.65 RCIM_Cubic>
         <at 1.3472 1 RCIM_Cubic>
         <at 1.5972 0.65>
         <at 1.8472 1 RCIM_Cubic>
        ]}}}
     CurveIndex=$System.NormalizedLoopAge
    )
    ScaleA=false
    ScaleRGB=true
   )
  ]
  rendererClasses [&NiagaraSpriteRendererProperties]
  renderers [{rendererClass=&NiagaraSpriteRendererProperties}]
 }
 Middle_Light {
  emitterUpdateScript [
   &EmitterState()
   &SpawnBurst_Instantaneous(Use Spawn Probability=false, UseLoopCountLimit=true)
  ]
  particleSpawnScript [
   &InitializeParticle(
    Color=<rgba 0.0252 0.4125 0.2016 1>
    Sprite Size Mode=::NewEnumerator3
    Uniform Sprite Size=40
    UsePositionOffset=false
    Write Material Random=true
   )
   &FN_Light_v2(Light Exponent=4, Light Radius=30, LightEnabled=true)
  ]
  particleUpdateScript [
   &ParticleState(Kill Particles When Lifetime Has Elapsed=false, Let Infinitely Lived Particles Die When Emitter Deactivates=true)
   &SkeletalMeshLocation(
    Bone Sampling Mode=::NewEnumerator4
    MaskSpawnGroup=false
    Sampled Orientation Coordinate Space=::World
    Sampled Position Coordinate Space=::World
    Sampled Position Offset=<0 -5 1>
    Sampled Velocity Coordinate Space:ENiagaraCoordinateSpace=::World
    Skeletal Mesh=$User.SkeletalMesh
   )
   /SetVariables(Particles.Light Color:LinearColor=&MultiplyLinearColor(Linear Color=<rgba 0.0252 0.4125 0.2016 1>, Scale Factor=&MakeVector4_FromFloat(Float=6))) {isSetParameters=true, module=SetVariables_#1}
  ]
  rendererClasses [&NiagaraLightRendererProperties]
  renderers [{rendererClass=&NiagaraLightRendererProperties, rendererIndex=0}]
 }
}
summary {
 systemName=NS_Backpack_CopperToll_Idle
 emitters [
  {emitterName=Middle_Glow, rendererClasses [&NiagaraSpriteRendererProperties]}
  {
   emitterName=Lantern_Light_Glow
   rendererClasses [&NiagaraLightRendererProperties, &NiagaraSpriteRendererProperties]
  }
  {emitterName=Middle_Light, rendererClasses [&NiagaraLightRendererProperties]}
 ]
 userVariables [
  User.bIsFrontEndPreview {type=NiagaraBool}
  User.SkeletalMesh {
   type=NiagaraDataInterfaceSkeletalMesh
   defaultValue {
    struct=NiagaraExt_VariableValue_DataInterface
    value {dataInterface=&self:NiagaraDataInterfaceSkeletalMesh_0}
   }
  }
 ]
}
