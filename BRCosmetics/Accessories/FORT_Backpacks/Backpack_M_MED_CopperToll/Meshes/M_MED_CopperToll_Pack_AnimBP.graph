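# Animation Blueprint for the CopperToll backpack cosmetic that adds secondary motion to its cloth, lanterns and coins.
# A chain of root-relative Anim Dynamics nodes simulates angular-limited cloth and lantern bones, then four coin bones on linear springs with locked rotation.
# The coin nodes use boolean alpha blending, shown fixed to true.
# Each update sets `coins turned on` false while skydiving or in the front end, though no visible node reads it.
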AnimBlueprint /BRCosmetics/Accessories/FORT_Backpacks/Backpack_M_MED_CopperToll/Meshes/M_MED_CopperToll_Pack_AnimBP
target_skeleton=&/M_MED_CopperToll_Pack_Skeleton
graphs [
 AnimGraph {nodes [
   n1 = Anim Dynamics - Bone: cloth_main(ComponentPose=Local To Component(LocalPose=Input Pose()), Alpha=1, ExternalForce="0,0,0") {
    angularDampingOverride=0.99
    angularSpringConstant=5000
    bOverrideAngularDamping=true
    bUseGravityOverride=true
    gravityScale=0.56
    linearDampingOverride=1
    simulationSpace=RootRelative
    alphaBoolBlend {blendOption=Linear}
    boundBone {boneName=cloth_main}
    componentLinearAccScale=<0.4 0.4 0.35>
    componentLinearVelScale=<0.5 0.5 0.5>
    gravityOverride {z=-730}
    physicsBodyDefinitions [{
      boundBone {boneName=cloth_main}
      boxExtents=<12 14 14>
      constraintSetup {angularLimitsMax=<0 70 25>, angularLimitsMin=<0 0 -25>, angularTarget=<0 0 0>}
      localJointOffset=<2 0 0>
     }]
   }
   n2 = Anim Dynamics - Bone: cloth_00(ComponentPose=@n1, Alpha=1, ExternalForce="0,0,0") {
    angularDampingOverride=0.99
    angularSpringConstant=5000
    bOverrideAngularDamping=true
    bUseGravityOverride=true
    gravityScale=0.56
    linearDampingOverride=1
    simulationSpace=RootRelative
    alphaBoolBlend {blendOption=Linear}
    boundBone {boneName=cloth_00}
    componentLinearAccScale=<0.25 0.25 0.28>
    componentLinearVelScale=<0.4 0.4 0.4>
    gravityOverride {z=-980}
    physicsBodyDefinitions [{
      boundBone {boneName=cloth_00}
      boxExtents=<16 16 16>
      constraintSetup {angularLimitsMax=<0 50 25>, angularLimitsMin=<0 0 -25>, angularTarget=<0 0 0>}
      localJointOffset=<2 0 0>
     }]
   }
   n3 = Anim Dynamics - Bone: cloth_02(ComponentPose=@n2, Alpha=1, ExternalForce="0,0,0") {
    angularDampingOverride=0.99
    angularSpringConstant=5000
    bOverrideAngularDamping=true
    bUseGravityOverride=true
    gravityScale=0.56
    linearDampingOverride=1
    simulationSpace=RootRelative
    alphaBoolBlend {blendOption=Linear}
    boundBone {boneName=cloth_02}
    componentLinearAccScale=<0.4 0.4 0.4>
    componentLinearVelScale=<0.5 0.5 0.5>
    gravityOverride {z=-980}
    physicsBodyDefinitions [{
      boundBone {boneName=cloth_02}
      boxExtents=<10 15 15>
      constraintSetup {angularLimitsMax=<0 90 30>, angularLimitsMin=<0 0 -30>, angularTarget=<0 0 0>}
      localJointOffset=<2 0 0>
     }]
   }
   n4 = Anim Dynamics - Bone: lantern_01(ComponentPose=@n3, Alpha=1, ExternalForce="0,0,0") {
    angularDampingOverride=1
    bGravityOverrideInSimSpace=true
    bOverrideAngularDamping=true
    bUseGravityOverride=true
    bUsePlanarLimit=false
    simulationSpace=RootRelative
    alphaBoolBlend {blendOption=Linear}
    boundBone {boneName=lantern_01}
    componentLinearAccScale=<0.2 0.2 0.3>
    componentLinearVelScale=<0.3 0.3 0.4>
    gravityOverride {z=-980}
    physicsBodyDefinitions [{
      sphereCollisionRadius=0
      boundBone {boneName=lantern_01}
      boxExtents=<10 10 10>
      constraintSetup {angularLimitsMax=<25 45 60>, angularLimitsMin=<-25 -45 -15>, angularTarget=<0 0 0>}
      localJointOffset=<3 0 0>
     }]
   }
   n5 = Anim Dynamics - Bone: lantern_02(ComponentPose=@n4, Alpha=1, ExternalForce="0,0,0") {
    angularDampingOverride=1
    bGravityOverrideInSimSpace=true
    bOverrideAngularDamping=true
    bUseGravityOverride=true
    bUsePlanarLimit=false
    simulationSpace=RootRelative
    alphaBoolBlend {blendOption=Linear}
    boundBone {boneName=lantern_02}
    componentLinearAccScale=<0.2 0.2 0.3>
    componentLinearVelScale=<0.3 0.3 0.4>
    gravityOverride {z=-980}
    physicsBodyDefinitions [{
      sphereCollisionRadius=0
      boundBone {boneName=lantern_02}
      boxExtents=<10 10 10>
      constraintSetup {angularLimitsMax=<25 45 60>, angularLimitsMin=<-25 -45 -15>, angularTarget=<0 0 0>}
      localJointOffset=<3 0 0>
     }]
   }
   n6 = Anim Dynamics - Bone: coin_01!(ComponentPose=@n5, bAlphaBoolEnabled:bool=True, ExternalForce="0,0,0") -> (Pose) {
    alphaInputType=Bool
    bLinearSpring=true
    bOverrideLinearDamping=true
    bUseGravityOverride=true
    linearDampingOverride=1
    linearSpringConstant=175
    simulationSpace=RootRelative
    alphaBoolBlend {blendInTime=0.2, blendOption=Linear, blendOutTime=0.2}
    boundBone {boneName=coin_01}
    componentAppliedLinearAccClamp=<2500 2500 2500>
    componentLinearAccScale=<0.1 0.1 0.3>
    componentLinearVelScale=<0.1 0.1 1>
    gravityOverride {z=-600}
    physicsBodyDefinitions [{
      boundBone {boneName=coin_01}
      boxExtents=<80 80 60>
      constraintSetup {angularLimitsMax=<0 0 0>, angularLimitsMin=<0 0 0>, angularTarget=<0 0 0>, linearAxesMax=<1 0.5 9>, linearAxesMin=<-1 -0.5 0>}
      localJointOffset=<0 0 0>
     }]
   }
   n7 = Anim Dynamics - Bone: coin_02!(ComponentPose=@n6, bAlphaBoolEnabled:bool=True, ExternalForce="0,0,0") -> (Pose) {
    alphaInputType=Bool
    bLinearSpring=true
    bOverrideLinearDamping=true
    bUseGravityOverride=true
    linearDampingOverride=1
    linearSpringConstant=175
    simulationSpace=RootRelative
    alphaBoolBlend {blendInTime=0.2, blendOption=Linear, blendOutTime=0.2}
    boundBone {boneName=coin_02}
    componentAppliedLinearAccClamp=<2500 2500 2500>
    componentLinearAccScale=<0.1 0.1 0.2>
    componentLinearVelScale=<0.1 0.1 0.9>
    gravityOverride {z=-700}
    physicsBodyDefinitions [{
      boundBone {boneName=coin_02}
      boxExtents=<80 80 60>
      constraintSetup {angularLimitsMax=<0 0 0>, angularLimitsMin=<0 0 0>, angularTarget=<0 0 0>, linearAxesMax=<1 0 5>, linearAxesMin=<-1 0 0>}
      localJointOffset=<0 0 0>
     }]
   }
   n8 = Anim Dynamics - Bone: coin_03!(ComponentPose=@n7, bAlphaBoolEnabled:bool=True, ExternalForce="0,0,0") -> (Pose) {
    alphaInputType=Bool
    bLinearSpring=true
    bOverrideLinearDamping=true
    bUseGravityOverride=true
    linearDampingOverride=1
    linearSpringConstant=175
    simulationSpace=RootRelative
    alphaBoolBlend {blendInTime=0.2, blendOption=Linear, blendOutTime=0.2}
    boundBone {boneName=coin_03}
    componentAppliedLinearAccClamp=<2500 2500 2500>
    componentLinearAccScale=<0.1 0.1 0.2>
    componentLinearVelScale=<0.1 0.1 0.9>
    gravityOverride {z=-500}
    physicsBodyDefinitions [{
      boundBone {boneName=coin_03}
      boxExtents=<80 80 60>
      constraintSetup {angularLimitsMax=<0 0 0>, angularLimitsMin=<0 0 0>, angularTarget=<0 0 0>, linearAxesMax=<0.5 0 4>, linearAxesMin=<-0.5 0 0>}
      localJointOffset=<0 0 0>
     }]
   }
   n9 = Anim Dynamics - Bone: coin_04!(ComponentPose=@n8, bAlphaBoolEnabled:bool=True, ExternalForce="0,0,0") -> (Pose) {
    alphaInputType=Bool
    bLinearSpring=true
    bOverrideLinearDamping=true
    bUseGravityOverride=true
    linearDampingOverride=1
    linearSpringConstant=175
    simulationSpace=RootRelative
    alphaBoolBlend {blendInTime=0.2, blendOption=Linear, blendOutTime=0.2}
    boundBone {boneName=coin_04}
    componentAppliedLinearAccClamp=<2500 2500 2500>
    componentLinearAccScale=<0.1 0.1 0.2>
    componentLinearVelScale=<0.1 0.1 0.9>
    gravityOverride {z=-550}
    physicsBodyDefinitions [{
      boundBone {boneName=coin_04}
      boxExtents=<80 80 60>
      constraintSetup {angularLimitsMax=<0 0 0>, angularLimitsMin=<0 0 0>, angularTarget=<0 0 0>, linearAxesMax=<0.5 0.5 2>, linearAxesMin=<-0.5 -0.5 0>}
      localJointOffset=<0 0 0>
     }]
   }
   Output Pose(Result=Component To Local(ComponentPose=@n9)) {layerGroup=None}
  ]}
 EventGraph {nodes [
   Event BlueprintUpdateAnimation()
   n1 = Get bIsSkyDiving()
   n2 = Get bIsInFrontEnd()
   > Set coins turned on(coins turned on:bool=NOT Boolean(A=OR Boolean(A=@n1, B=@n2))) -> (Output_Get:bool)
   TryGetPawnOwner()
  ]}
]
variables [coins turned on {type=boolean}]
